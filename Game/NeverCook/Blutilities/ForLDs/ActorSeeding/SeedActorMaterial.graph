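# SeedActorMaterial is an unlit, translucent, two-sided material in the level-designer actor seeding Blutility folder, apparently an editor visual for seed actors.
# A Color vector parameter supplies a green tint.
# CameraDepthFade uses a negative length to invert the fade, with its offset driven by the FadeDistance parameter plus a constant.
# The fade is clamped to a very low maximum, keeping the value faint, then feeds an Add node.
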Material /Game/NeverCook/Blutilities/ForLDs/ActorSeeding/SeedActorMaterial
graph {
 properties {blendMode=BLEND_Translucent, pixelDepthOffsetMode=PDOM_Legacy, shadingModel=MSM_Unlit, twoSided=true, usedShadingModels=Unlit}
 nodes [
  n1 = Constant() {r=-8000}
  n2 = Add(A=ScalarParameter() {parameterName=FadeDistance}, B=Constant() {r=10000})
  Add(B=Clamp(None=MaterialFunctionCall(Fade Length=@n1, Fade Offset=@n2) {materialFunction=&CameraDepthFade}.Result) {maxDefault=0.025})
  VectorParameter() {parameterName=Color, defaultValue=<rgba 0.2033 1 0.4447 1>}
 ]
}
