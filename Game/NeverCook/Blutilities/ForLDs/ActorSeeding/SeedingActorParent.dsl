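; Level-designer seeding actor that scatters actors over landscape within a radius and destroys itself at BeginPlay.
; Toggling validates data, clears prior spawns, places a first actor, then seeds primary, secondary and tertiary tiers through recursive retries.
; Candidates offset randomly from the previous spawn, trace down, and pass water, physical material, slope, radius and spacing checks.
; Spawning applies chance, ordering, scale, alignment and tilt; the construction script colors an editor volume by validity.
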(event Custom|DebugPrints
  (params (Reason :type String) (Color :type Linear%20Color%20Structure :default "(R=0.421875,G=0.614589,B=1.000000,A=1.000000)") (Duration :type Float%20%28double-precision%29 :default 5.0) (PrintToScreen :type Boolean))
  ;[1 1 1] Debug
  (if (Variables|Debug|GetEnableDebugLogging)
    (Development|PrintString (Utilities|String|Append "Seed Actor Debug //  " Reason) PrintToScreen true Color Duration)))

(event EventBeginPlay
  (bind _self self)
  ;[1 1 1] Cleanup if forgot to delete ...
  (Utilities|IsValid _self
    (:Is%20Valid
      (Actor|DestroyActor _self))))

(event Custom|Clear%20Seeded%20Actors
  ;[1 1 1] Remove Spawned Actors
  (CallFunction|ClearSpawnedActors))

(event Custom|RetryInitialPlacement
  ;[1 1 1] Actor Seeding // Initial Actor
  ;;[1 1 1] Spawn the first primary actor so actors can be clustered
  (bind _succeeded (CallFunction|SpawnFirstActor))
  (if _succeeded
    (CallFunction|SeedPrimaryActors)
    (else
      ;;[1 1 1] RetryInitialPlacementAttemps
      (bind _result (Math|Integer|IncrementInt (Variables|InternalVariables|GetRetryInitialPlacementCount))
        (:%20%20
          (if (>= _result (Variables|InternalVariables|GetMaxInitialPlacementAttemps))
            (CallFunction|DebugPrints "Failed to spawn the first actor after retrying 50 times, something went really wrong ;(" "(R=1.000000,G=0.045985,B=0.000000,A=1.000000)" 1e+01 true)
            (else
              (CallFunction|RetryInitialPlacement))))))))

(event Custom|Toggle%20Seeding%20Actors
  ;[1 1 1] Actor Seeding // Initial Actor
  ;;[1 1 1] Is data valid?
  (bind _valid (CallFunction|ValidateData))
  (if _valid
    (CallFunction|ClearSeededActors)
    (CallFunction|GetAllActorsinSeedArea) ; cache dynamic actors in seed area
    ;;[1 1 1] Spawn the first primary actor so actors can be clustered
    (bind _succeeded (CallFunction|SpawnFirstActor))
    (if _succeeded
      (CallFunction|SeedPrimaryActors)
      (else
        ;;[1 1 1] RetryInitialPlacementAttemps
        (bind _result (Math|Integer|IncrementInt (Variables|InternalVariables|GetRetryInitialPlacementCount))
          (:%20%20
            (if (>= _result (Variables|InternalVariables|GetMaxInitialPlacementAttemps))
              (CallFunction|DebugPrints "Failed to spawn the first actor after retrying 50 times, something went really wrong ;(" "(R=1.000000,G=0.045985,B=0.000000,A=1.000000)" 1e+01 true)
              (else
                (CallFunction|RetryInitialPlacement)))))))
    (else
      (CallFunction|DebugPrints "Data is Invalid, check actors are set to spawn and quantity is set, will also fail if quantity of total actors to spawn is more than 300" "(R=1.000000,G=0.045985,B=0.000000,A=1.000000)" 1e+01))))

(event Custom|SeedPrimaryActors
  (bind (_seedingradius_2_816d77b142be0927697e518f6673b0f1 _numprimaryseedactors_15_a4812bae43a70605e6445396062315af) (Utilities|Struct|BreakActorSeedingProfileData (Variables|ActorSeeding|GetPlacementSeedingData)))
  ;[1 1 1] Seed Primary Actors
  (Variables|PrimarySeeding|SetNumPrimaryActorsToSeedRemaining (- _numprimaryseedactors_15_a4812bae43a70605e6445396062315af 1))
  (CallFunction|AttemptPrimaryActorSeed))

(event Custom|AttemptPrimaryActorSeed
  (bind _retryprimaryplacementcount (Variables|PrimarySeeding|GetRetryPrimaryPlacementCount))
  ;[1 1 1] Seed Primary Actors
  (if (<= (Variables|PrimarySeeding|GetNumPrimaryActorsToSeedRemaining) 0)
    (CallFunction|SeedSecondaryActors)
    (else
      (bind _succeeded (CallFunction|MainActorSeedingFunction (Variables|ActorSeeding|GetPrimaryActorSeedData)))
      (if _succeeded
        (Math|Integer|DecrementInt (Variables|PrimarySeeding|GetNumPrimaryActorsToSeedRemaining)
          (:%20%20
            (CallFunction|AttemptPrimaryActorSeed)))
        (elif (>= _retryprimaryplacementcount (Variables|PrimarySeeding|GetMaxNumPrimaryPlacementRetry))
          (CallFunction|SeedSecondaryActors)
          (else
            (Math|Integer|IncrementInt _retryprimaryplacementcount
              (:%20%20
                (CallFunction|AttemptPrimaryActorSeed)))))))))

(event Custom|SeedSecondaryActors
  (bind (_seedingradius_2_816d77b142be0927697e518f6673b0f1 _numprimaryseedactors_15_a4812bae43a70605e6445396062315af _numsecondaryseedactors_18_e5908d964e9d7e0f9b994786c762732e) (Utilities|Struct|BreakActorSeedingProfileData (Variables|ActorSeeding|GetPlacementSeedingData)))
  ;[1 1 1] Seed Secondary Actors
  (Variables|InternalVariables|SetCurrentActorToSpawnIndex -1)
  (if (> (Utilities|Array|Length (Utilities|Struct|BreakActorSeedData (Variables|ActorSeeding|GetSecondaryActorSeedData))) 0)
    (Variables|SecondarySeeding|SetNumSecondaryActorsToSeedRemaining _numsecondaryseedactors_18_e5908d964e9d7e0f9b994786c762732e)
    (CallFunction|AttemptSecondaryActorSeed)
    (else
      (CallFunction|SeedTertiaryActors))))

(event Custom|AttemptSecondaryActorSeed
  (bind _retrysecondaryplacementcount (Variables|SecondarySeeding|GetRetrySecondaryPlacementCount))
  ;[1 1 1] Seed Secondary Actors
  (if (<= (Variables|SecondarySeeding|GetNumSecondaryActorsToSeedRemaining) 0)
    (CallFunction|SeedTertiaryActors)
    (else
      (bind _succeeded (CallFunction|MainActorSeedingFunction (Variables|ActorSeeding|GetSecondaryActorSeedData) 1))
      (if _succeeded
        (Math|Integer|DecrementInt (Variables|SecondarySeeding|GetNumSecondaryActorsToSeedRemaining)
          (:%20%20
            (CallFunction|AttemptSecondaryActorSeed)))
        (elif (>= _retrysecondaryplacementcount (Variables|SecondarySeeding|GetMaxNumberSecondaryPlacementRetry))
          (CallFunction|SeedTertiaryActors)
          (else
            (Math|Integer|IncrementInt _retrysecondaryplacementcount
              (:%20%20
                (CallFunction|AttemptSecondaryActorSeed)))))))))

(event Custom|SeedTertiaryActors
  (bind (_seedingradius_2_816d77b142be0927697e518f6673b0f1 _numprimaryseedactors_15_a4812bae43a70605e6445396062315af _numsecondaryseedactors_18_e5908d964e9d7e0f9b994786c762732e _numtertiaryseedactors_19_b35a07a2453f88aeacf7499d9dd7543c) (Utilities|Struct|BreakActorSeedingProfileData (Variables|ActorSeeding|GetPlacementSeedingData)))
  ;[1 1 1] Seed Tertiary Actors
  (Variables|InternalVariables|SetCurrentActorToSpawnIndex -1)
  (if (> (Utilities|Array|Length (Utilities|Struct|BreakActorSeedData (Variables|ActorSeeding|GetTertiaryActorSeedData))) 0)
    (Variables|TertiarySeeding|SetNumTertiaryActorsToSeedRemaining _numtertiaryseedactors_19_b35a07a2453f88aeacf7499d9dd7543c)
    (CallFunction|AttemptTertiaryActorSeed)))

(event Custom|AttemptTertiaryActorSeed
  (bind (_seedingradius_2_816d77b142be0927697e518f6673b0f1 _numprimaryseedactors_15_a4812bae43a70605e6445396062315af _numsecondaryseedactors_18_e5908d964e9d7e0f9b994786c762732e _numtertiaryseedactors_19_b35a07a2453f88aeacf7499d9dd7543c) (Utilities|Struct|BreakActorSeedingProfileData (Variables|ActorSeeding|GetPlacementSeedingData)))
  (bind _retrytertiaryplacementcount (Variables|TertiarySeeding|GetRetryTertiaryPlacementCount))
  ;[1 1 1] Seed Tertiary Actors
  (if (<= (Variables|TertiarySeeding|GetNumTertiaryActorsToSeedRemaining) 0)
    (CallFunction|DebugPrints (Utilities|String|Append "All Seeding Actors Have Spawned, A Total of  " (Utilities|String|ToString(Integer) (Utilities|Array|Length (Variables|ActorRefs|GetAllSpawnedActorsRef))) "/" (Utilities|String|ToString(Integer) (Utilities|Operators|Add _numprimaryseedactors_15_a4812bae43a70605e6445396062315af _numsecondaryseedactors_18_e5908d964e9d7e0f9b994786c762732e _numtertiaryseedactors_19_b35a07a2453f88aeacf7499d9dd7543c)) " Possible Actors Have Been Placed.") "(R=0.225856,G=1.000000,B=0.000000,A=1.000000)" 3.0 true)
    (else
      (bind _succeeded (CallFunction|MainActorSeedingFunction (Variables|ActorSeeding|GetTertiaryActorSeedData) 2))
      (if _succeeded
        (Math|Integer|DecrementInt (Variables|TertiarySeeding|GetNumTertiaryActorsToSeedRemaining)
          (:%20%20
            (CallFunction|AttemptTertiaryActorSeed)))
        (elif (>= _retrytertiaryplacementcount (Variables|TertiarySeeding|GetMaxNumTertiaryPlacementRetry))
          _
          (else
            (Math|Integer|IncrementInt _retrytertiaryplacementcount
              (:%20%20
                (CallFunction|AttemptTertiaryActorSeed)))))))))

(fn ClearSpawnedActors
  (bind _allspawnedactorsref (Variables|ActorRefs|GetAllSpawnedActorsRef))
  (if (> (Utilities|Array|Length _allspawnedactorsref) 0)
    (for _array_element _allspawnedactorsref
      (Actor|DestroyActor _array_element))
    (Utilities|Array|Clear (Variables|ActorRefs|GetAllSpawnedActorsRef))
    (Variables|ActorRefs|SetPreviousSpawnedActorRef 0)
    (Utilities|Array|Clear (Variables|ActorRefs|GetAllDynamicActorsInSeedingRadius))
    (Variables|InternalVariables|SetRetryInitialPlacementCount 0)
    (Variables|PrimarySeeding|SetRetryPrimaryPlacementCount 0)
    (Variables|SecondarySeeding|SetRetrySecondaryPlacementCount 0)
    (Variables|TertiarySeeding|SetRetryTertiaryPlacementCount 0)))

(fn FindLocationForSeeding
  (params (ActorSeedData :type Actor%20Seed%20Data%20Structure) (RadiusToFind :type Float%20%28double-precision%29) (Actor :type Actor%20Object%20Reference))
  (bind (_actorstoseed_51_d7f231c2419ae55740a8b4bcaac1f7ca _mindistancetoselftype_7_01c82dcc4409c0b9f271018cdd0d768c _mindistancetobuildingactors_9_923c5abd4707bc47c299a0abab7a1749 _allowseedinwaterbodies_12_4f15c8b44c829f7843658484951353b0) (Utilities|Struct|BreakActorSeedData (LocalVariables|GetActorSeedDataLocal)))
  (LocalVariables|SetActorSeedDataLocal ActorSeedData)
  (bind _postoffsetmod (CallFunction|RandomOffsetXY (Math|Vector|MakeVector (.x (Transformation|GetActorLocation Actor)) (.y (Transformation|GetActorLocation Actor)) (+ (.z (Transformation|GetActorLocation Actor)) (/ (Variables|ActorSeedingAdvanced|GetZProjectionDistance) 2.0))) RadiusToFind))
  (LocalVariables|SetLocalPostModVector _postoffsetmod)
  ;[1 1 1] Waterbody trace check
  (bind _seedinwater (CallFunction|ShouldSeedinWater (LocalVariables|GetActorSeedDataLocal)))
  (if (not _seedinwater)
    (bind _outhit (Collision|LineTraceByChannel :Start (LocalVariables|GetLocalPostModVector) :End (- (LocalVariables|GetLocalPostModVector) (Math|Vector|MakeVector 0.0 0.0 (Variables|ActorSeedingAdvanced|GetZProjectionDistance))) :TraceChannel "TraceTypeQuery12" :bTraceComplex false :DrawDebugType (Variables|Debug|GetDebugDrawLineTrace) :bIgnoreSelf true :TraceColor "(R=1.000000,G=0.000000,B=0.000000,A=1.000000)" :TraceHitColor "(R=0.000000,G=1.000000,B=0.000000,A=1.000000)" :DrawTime (Variables|Debug|GetDebugDrawLineDuration)))
    (if (Collision|BreakHitResult _outhit)
      (LocalVariables|SetWaterVolumeHitLocationLocal (.Location (Collision|BreakHitResult _outhit))) ; store this to compare to height of landscape hit, because ocean is below landscape etc
      (LocalVariables|SetWaterbodyHitLocal true)
      (else
        _))
    ;[1 1 1] World Trace for Landscape
    (bind _outhits (Collision|MultiLineTraceByChannel :Start (LocalVariables|GetLocalPostModVector) :End (- (LocalVariables|GetLocalPostModVector) (Math|Vector|MakeVector 0.0 0.0 (Variables|ActorSeedingAdvanced|GetZProjectionDistance))) :TraceChannel "TraceTypeQuery10" :bTraceComplex true :DrawDebugType (Variables|Debug|GetDebugDrawLineTrace) :bIgnoreSelf true :TraceColor "(R=1.000000,G=0.000000,B=0.000000,A=1.000000)" :TraceHitColor "(R=0.000000,G=1.000000,B=0.000000,A=1.000000)" :DrawTime (Variables|Debug|GetDebugDrawLineDuration)))
    (bind _array_element (Utilities|Array|ForEachLoopwithBreak _outhits)
      (:LoopBody
        (if (Collision|BreakHitResult _array_element)
          (if (== (Utilities|GetClass (.HitActor (Collision|BreakHitResult _array_element))) "/Script/Landscape.LandscapeStreamingProxy")
            (bind _allow (CallFunction|PhysicsMaterialSpawnCheck (LocalVariables|GetActorSeedDataLocal) (Utilities|GetDisplayName (.PhysMat (Collision|BreakHitResult _array_element)))))
            (if _allow
              (LocalVariables|SetTraceLandscapeHitLocationLocal (.Location (Collision|BreakHitResult _array_element)))
              (LocalVariables|SetTraceLandscapeHitNormalLocal (.Normal (Collision|BreakHitResult _array_element)))
              (LocalVariables|SetHitLandscapeLocal true)))
          (if (== (Utilities|GetClass (.HitActor (Collision|BreakHitResult _array_element))) "/Script/Landscape.Landscape")
            _)
          (Utilities|Casting|CastToFortStaticMeshActorClass (Utilities|GetClass (.HitActor (Collision|BreakHitResult _array_element))))))
      (:Completed
        _))
    ;[1 1 1] Water/Slope/Seeding Radius Check
    (if (LocalVariables|GetHitLandscapeLocal)
      (if (not _allowseedinwaterbodies_12_4f15c8b44c829f7843658484951353b0)
        (bind _allowspawn (CallFunction|WaterAboveBelowCheck (LocalVariables|GetTraceLandscapeHitLocationLocal) (LocalVariables|GetWaterVolumeHitLocationLocal)))
        (if _allowspawn
          (bind (_slopeisvalid _lookatrotation _slopeissteep) (Default|IsSlopeValid (LocalVariables|GetTraceLandscapeHitNormalLocal) (LocalVariables|GetTraceLandscapeHitLocationLocal) (LocalVariables|GetActorSeedDataLocal)))
          (if _slopeisvalid
            (LocalVariables|SetLocalLookAtRot _lookatrotation) ; used for align to normal
            (bind _valid (CallFunction|IsWithinSeedingRadius (LocalVariables|GetTraceLandscapeHitLocationLocal)))
            (if _valid
              (LocalVariables|SetSearchSucceededLocal true)
              (else
                _))
            (else
              _))
          (return (LocalVariables|GetSearchSucceededLocal) (LocalVariables|GetTraceLandscapeHitLocationLocal) (LocalVariables|GetTraceLandscapeHitNormalLocal) (LocalVariables|GetLocalLookAtRot) _slopeissteep)
          (else
            (CallFunction|DebugPrints "Water is detected above terrain and actor is set to not be spawned in waterbodies"))))
      (else
        _))))

(fn GetAllActorsInSeedArea
  (bind _outactors (Collision|SphereOverlapActors (Transformation|GetActorLocation self) (Utilities|Struct|BreakActorSeedingProfileData (Variables|ActorSeeding|GetPlacementSeedingData)) (Utilities|Array|MakeArray "ObjectTypeQuery1") "/Script/Engine.Actor"))
  (Variables|ActorRefs|SetAllDynamicActorsInSeedingRadius _outactors))

(fn GetClusteringRadius
  (params (ActorSeedingData :type Actor%20Seed%20Data%20Structure) (SeedingStage :type Integer))
  (bind _placementseedingdata (Variables|ActorSeeding|GetPlacementSeedingData))
  (bind (_seedingradius_2_816d77b142be0927697e518f6673b0f1 _numprimaryseedactors_15_a4812bae43a70605e6445396062315af _numsecondaryseedactors_18_e5908d964e9d7e0f9b994786c762732e _numtertiaryseedactors_19_b35a07a2453f88aeacf7499d9dd7543c _primaryclusteringprofile_28_ff3be676454423c0253f868968b484a1 _secondaryclusteringprofile_31_f6dd305f47adc1826540b5b0ffc0e95d _tertiaryclusteringprofile_33_0c6a1d454d309264fc4b3db74bd9d7b4) (Utilities|Struct|BreakActorSeedingProfileData _placementseedingdata))
  (bind (_seedingradius_2_816d77b142be0927697e518f6673b0f1_2 _numprimaryseedactors_15_a4812bae43a70605e6445396062315af_2 _numsecondaryseedactors_18_e5908d964e9d7e0f9b994786c762732e_2 _numtertiaryseedactors_19_b35a07a2453f88aeacf7499d9dd7543c_2 _primaryclusteringprofile_28_ff3be676454423c0253f868968b484a1_2) (Utilities|Struct|BreakActorSeedingProfileData (Variables|ActorSeeding|GetPlacementSeedingData)))
  (bind (_seedingradius_2_816d77b142be0927697e518f6673b0f1_3 _numprimaryseedactors_15_a4812bae43a70605e6445396062315af_3 _numsecondaryseedactors_18_e5908d964e9d7e0f9b994786c762732e_3 _numtertiaryseedactors_19_b35a07a2453f88aeacf7499d9dd7543c_3 _primaryclusteringprofile_28_ff3be676454423c0253f868968b484a1_3 _secondaryclusteringprofile_31_f6dd305f47adc1826540b5b0ffc0e95d_2) (Utilities|Struct|BreakActorSeedingProfileData (Variables|ActorSeeding|GetPlacementSeedingData)))
  (bind (_actorstoseed_51_d7f231c2419ae55740a8b4bcaac1f7ca _mindistancetoselftype_7_01c82dcc4409c0b9f271018cdd0d768c) (Utilities|Struct|BreakActorSeedData (LocalVariables|GetSeedDataLocal)))
  (bind (_actorstoseed_51_d7f231c2419ae55740a8b4bcaac1f7ca_2 _mindistancetoselftype_7_01c82dcc4409c0b9f271018cdd0d768c_2) (Utilities|Struct|BreakActorSeedData (LocalVariables|GetSeedDataLocal)))
  (LocalVariables|SetSeedDataLocal ActorSeedingData)
  (switch int SeedingStage
    (:0
      (switch Utilities|FlowControl|Switch|SwitchonActorSeedingClusteringProfiles _primaryclusteringprofile_28_ff3be676454423c0253f868968b484a1_2
        (:NewEnumerator0
          (LocalVariables|SetFinalRadiusLocal (* _mindistancetoselftype_7_01c82dcc4409c0b9f271018cdd0d768c (Math|Random|RandomFloatinRange 2.0 3.5))))
        (:NewEnumerator1
          (LocalVariables|SetFinalRadiusLocal (* _mindistancetoselftype_7_01c82dcc4409c0b9f271018cdd0d768c_2 (Math|Random|RandomFloatinRange 3.0 5.0))))
        (:NewEnumerator2
          (LocalVariables|SetFinalRadiusLocal _seedingradius_2_816d77b142be0927697e518f6673b0f1)))
      (return (LocalVariables|GetFinalRadiusLocal)))
    (:1
      (switch Utilities|FlowControl|Switch|SwitchonActorSeedingClusteringProfiles _secondaryclusteringprofile_31_f6dd305f47adc1826540b5b0ffc0e95d_2
        (:NewEnumerator0
          _)
        (:NewEnumerator1
          _)
        (:NewEnumerator2
          _)))
    (:2
      (switch Utilities|FlowControl|Switch|SwitchonActorSeedingClusteringProfiles _tertiaryclusteringprofile_33_0c6a1d454d309264fc4b3db74bd9d7b4
        (:NewEnumerator0
          _)
        (:NewEnumerator1
          _)
        (:NewEnumerator2
          _)))))

(fn IsNearAnyBuildingActor)

(fn IsNearSelfActorType)

(fn IsSlopeValid
  (params (HitNormal :type Vector) (HitLocation :type Vector) (ActorSeedData :type Actor%20Seed%20Data%20Structure))
  (bind (_actorstoseed_51_d7f231c2419ae55740a8b4bcaac1f7ca _mindistancetoselftype_7_01c82dcc4409c0b9f271018cdd0d768c _mindistancetobuildingactors_9_923c5abd4707bc47c299a0abab7a1749 _allowseedinwaterbodies_12_4f15c8b44c829f7843658484951353b0 _seedzoffset_15_022bf3f04c9d45abd279a58dd7d53d08 _actormaxslopeangle_18_74ca4fea488b5facee55608578cf6572) (Utilities|Struct|BreakActorSeedData ActorSeedData))
  (bind _returnvalue (* HitNormal "400.000000,400.000000,400.000000"))
  (bind _returnvalue_2 (+ _returnvalue HitLocation))
  (bind _returnvalue_3 (Math|Rotator|FindLookatRotation HitLocation _returnvalue_2))
  (bind _roll (.roll _returnvalue_3))
  (bind _pitch (.pitch _returnvalue_3))
  (bind _maxslopelocal (LocalVariables|GetMaxSlopeLocal))
  (LocalVariables|SetMaxSlopeLocal _actormaxslopeangle_18_74ca4fea488b5facee55608578cf6572)
  (if (< _roll _maxslopelocal)
    (if (< _pitch _maxslopelocal)
      (return false _returnvalue_3 false)
      (else
        (LocalVariables|SetDetectedDegreeLocal _pitch)))
    (else
      (LocalVariables|SetDetectedDegreeLocal _roll)))
  (return true _returnvalue_3 (< (- (LocalVariables|GetDetectedDegreeLocal) (Variables|InternalVariables|GetExtremeSlopeNegaiveOffsetThreshold)) 3e+01)))

(fn IsWithinSeedingRadius
  (params (Hit :type Vector))
  (bind _distance (Class|MissionBlueprintFunctionLibrary|DistanceBetweenTwoVectors Hit (Transformation|GetActorLocation self)))
  (return (<= (Math|Float|Absolute(Float) _distance) (Utilities|Struct|BreakActorSeedingProfileData (Variables|ActorSeeding|GetPlacementSeedingData)))))

(fn MainActorSeedingFunction
  (params (ActorSeedingData :type Actor%20Seed%20Data%20Structure) (SeedingStage :type Integer))
  (bind _radius (CallFunction|GetClusteringRadius ActorSeedingData SeedingStage))
  (bind (_succeeded _validhitlocation _validhitnormal _lookatrotation _onsteepslope) (CallFunction|FindLocationforSeeding ActorSeedingData _radius (Variables|ActorRefs|GetPreviousSpawnedActorRef)))
  (bind _output_get (LocalVariables|SetSucceededLocal _succeeded))
  (if _output_get
    (bind _validforspawn (CallFunction|ProximityandOverlapCheck _validhitlocation ActorSeedingData))
    (if _validforspawn
      (CallFunction|SpawnActorinWorld _validhitlocation _validhitnormal ActorSeedingData _lookatrotation _onsteepslope)
      (else
        _))
    (return (LocalVariables|GetSucceededLocal))
    (else
      _)))

(fn PhysicsMaterialSpawnCheck
  (params (SeedData :type Actor%20Seed%20Data%20Structure) (DisplayNameOfHitPhysMat :type String))
  (bind (_actorstoseed_51_d7f231c2419ae55740a8b4bcaac1f7ca _mindistancetoselftype_7_01c82dcc4409c0b9f271018cdd0d768c _mindistancetobuildingactors_9_923c5abd4707bc47c299a0abab7a1749 _allowseedinwaterbodies_12_4f15c8b44c829f7843658484951353b0 _seedzoffset_15_022bf3f04c9d45abd279a58dd7d53d08 _actormaxslopeangle_18_74ca4fea488b5facee55608578cf6572 _manualoverrideminactorscale_24_cd532bb84c6429e2b0cb79910b71f3b6 _manualoverridemaxactorscale_25_7beaf98440627e0b3eb6aa9af5f5a1c7 _usemanualoverridescaling_27_8750023c414571fe715d07879086ddd8 _physicsmaterialignorelist_32_f5b884bb420973bacbce9ba20ffcc3b8) (Utilities|Struct|BreakActorSeedData SeedData))
  (if (Variables|Debug|GetEnableDebugPhysicsMaterials)
    (Development|PrintString (Utilities|String|Append "Physical Material Hit = " DisplayNameOfHitPhysMat) false))
  (if (> (Utilities|Array|Length _physicsmaterialignorelist_32_f5b884bb420973bacbce9ba20ffcc3b8) 0)
    (if (Utilities|Array|ContainsItem _physicsmaterialignorelist_32_f5b884bb420973bacbce9ba20ffcc3b8 DisplayNameOfHitPhysMat)
      (return false)
      (else
        (return true)))
    (else
      (return true))))

(fn ProximityAndOverlapCheck
  (params (HitLocation :type Vector) (SeedActorData :type Actor%20Seed%20Data%20Structure))
  (bind (_actorstoseed_51_d7f231c2419ae55740a8b4bcaac1f7ca _mindistancetoselftype_7_01c82dcc4409c0b9f271018cdd0d768c _mindistancetobuildingactors_9_923c5abd4707bc47c299a0abab7a1749) (Utilities|Struct|BreakActorSeedData SeedActorData))
  (bind _outactors (Collision|SphereOverlapActors HitLocation _mindistancetobuildingactors_9_923c5abd4707bc47c299a0abab7a1749 (Utilities|Array|MakeArray "ObjectTypeQuery1") "/Script/FortniteGame.BuildingActor" (Utilities|Array|MakeArray self)))
  (bind _array_element (Utilities|Array|ForEachLoopwithBreak _outactors)
    (:LoopBody
      (bind _distance (Class|MissionBlueprintFunctionLibrary|DistanceBetweenTwoVectors (Transformation|GetActorLocation _array_element) HitLocation))
      (if (< _distance _mindistancetobuildingactors_9_923c5abd4707bc47c299a0abab7a1749)
        (LocalVariables|SetFailedDistanceCheckLocal true)))
    (:Completed
      _))
  (if (LocalVariables|GetFailedDistanceCheckLocal)
    (return false)
    (else
      (bind _array_element_2 (Utilities|Array|ForEachLoopwithBreak (Variables|ActorRefs|GetAllDynamicActorsInSeedingRadius))
        (:LoopBody
          (if (== (Utilities|GetClass _array_element_2) (Utilities|Array|Get(acopy) _actorstoseed_51_d7f231c2419ae55740a8b4bcaac1f7ca))
            (bind _distance_2 (Class|MissionBlueprintFunctionLibrary|DistanceBetweenTwoVectors HitLocation (Transformation|GetActorLocation _array_element_2)))
            (if (< _distance_2 _mindistancetoselftype_7_01c82dcc4409c0b9f271018cdd0d768c)
              (LocalVariables|SetIsTooCloseLocal true))))
        (:Completed
          _))
      (if (LocalVariables|GetIsTooCloseLocal)
        (return false)
        (else
          (return true))))))

(fn RandomOffsetXY
  (params (PreOffsetMod :type Vector) (Offset :type Float%20%28double-precision%29))
  (return (Math|Vector|MakeVector (+ (.x PreOffsetMod) (Math|Random|RandomFloatinRange (* Offset -1.0) Offset)) (+ (.y PreOffsetMod) (Math|Random|RandomFloatinRange (* Offset -1.0) Offset)) (.z PreOffsetMod))))

(fn RandomTilt
  (params (RotIn :type Rotator) (TiltRange :type Float%20%28double-precision%29))
  (bind _output_get (LocalVariables|SetRotLocal RotIn))
  (bind _output_get_2 (LocalVariables|SetFloatLocalMax (Math|Float|Absolute(Float) TiltRange)))
  (LocalVariables|SetFloatLocalMin (* _output_get_2 -1.0))
  (bind _output_get_3 (LocalVariables|SetRotLocal (Math|Rotator|MakeRotator (Math|Random|RandomFloatinRange (LocalVariables|GetFloatLocalMin) (LocalVariables|GetFloatLocalMax)) (Math|Random|RandomFloatinRange (LocalVariables|GetFloatLocalMin) (LocalVariables|GetFloatLocalMax)) (.yaw _output_get))))
  (return _output_get_3))

(fn SetMaterialInstanceDynamic
  (params (Actor :type Actor%20Object%20Reference) (MaterialInstance :type Material%20Instance%20Object%20Reference))
  (Utilities|IsValid MaterialInstance
    (:Is%20Valid
      _)
    (:Is%20Not%20Valid
      _))
  (return))

(fn ShouldSeedInWater
  (params (ActorSeedData :type Actor%20Seed%20Data%20Structure))
  (bind (_actorstoseed_51_d7f231c2419ae55740a8b4bcaac1f7ca _mindistancetoselftype_7_01c82dcc4409c0b9f271018cdd0d768c _mindistancetobuildingactors_9_923c5abd4707bc47c299a0abab7a1749 _allowseedinwaterbodies_12_4f15c8b44c829f7843658484951353b0) (Utilities|Struct|BreakActorSeedData ActorSeedData))
  (LocalVariables|SetAllowSeedLocal _allowseedinwaterbodies_12_4f15c8b44c829f7843658484951353b0)
  (return (LocalVariables|GetAllowSeedLocal)))

(fn SpawnActorInWorld
  (params (LocationToSpawn :type Vector) (HitNormal :type Vector) (ActorSeedData :type Actor%20Seed%20Data%20Structure) (LookAtRotation :type Rotator) (IsOnSteepSlope :type Boolean))
  (bind (_actorstoseed_51_d7f231c2419ae55740a8b4bcaac1f7ca _mindistancetoselftype_7_01c82dcc4409c0b9f271018cdd0d768c _mindistancetobuildingactors_9_923c5abd4707bc47c299a0abab7a1749 _allowseedinwaterbodies_12_4f15c8b44c829f7843658484951353b0 _seedzoffset_15_022bf3f04c9d45abd279a58dd7d53d08 _actormaxslopeangle_18_74ca4fea488b5facee55608578cf6572 _manualoverrideminactorscale_24_cd532bb84c6429e2b0cb79910b71f3b6 _manualoverridemaxactorscale_25_7beaf98440627e0b3eb6aa9af5f5a1c7 _usemanualoverridescaling_27_8750023c414571fe715d07879086ddd8 _physicsmaterialignorelist_32_f5b884bb420973bacbce9ba20ffcc3b8 _aligntolandscapenormal_37_c6495313470a8b682c3d00b3f46aa7eb _usesteepslopenegativeoffset_41_0acea6d44064781e182520a011991d2f _actorseedingchance_45_2e134bd5489fa9e2966073a71e6d1185 _actorseedingspawnmethod_50_2043d7544f2ab551d89eeab6b67d10f1 _actorzoffsetrandomrange_55_e15a9b6149de31b18cc5198a8b6c5631 _actorsmaterialinstanceoverride_60_a5b63aac4ca0d3d831d255acac335632 _actortiltdegreesrange_63_950ce81f4955d3a2b7c87cabef554dad) (Utilities|Struct|BreakActorSeedData ActorSeedData))
  (bind (_actorstoseed_51_d7f231c2419ae55740a8b4bcaac1f7ca_2 _mindistancetoselftype_7_01c82dcc4409c0b9f271018cdd0d768c_2 _mindistancetobuildingactors_9_923c5abd4707bc47c299a0abab7a1749_2 _allowseedinwaterbodies_12_4f15c8b44c829f7843658484951353b0_2 _seedzoffset_15_022bf3f04c9d45abd279a58dd7d53d08_2 _actormaxslopeangle_18_74ca4fea488b5facee55608578cf6572_2 _manualoverrideminactorscale_24_cd532bb84c6429e2b0cb79910b71f3b6_2 _manualoverridemaxactorscale_25_7beaf98440627e0b3eb6aa9af5f5a1c7_2 _usemanualoverridescaling_27_8750023c414571fe715d07879086ddd8_2 _physicsmaterialignorelist_32_f5b884bb420973bacbce9ba20ffcc3b8_2 _aligntolandscapenormal_37_c6495313470a8b682c3d00b3f46aa7eb_2 _usesteepslopenegativeoffset_41_0acea6d44064781e182520a011991d2f_2 _actorseedingchance_45_2e134bd5489fa9e2966073a71e6d1185_2 _actorseedingspawnmethod_50_2043d7544f2ab551d89eeab6b67d10f1_2 _actorzoffsetrandomrange_55_e15a9b6149de31b18cc5198a8b6c5631_2 _actorsmaterialinstanceoverride_60_a5b63aac4ca0d3d831d255acac335632_2) (Utilities|Struct|BreakActorSeedData ActorSeedData))
  (bind _localrandommaxoffset (LocalVariables|GetLocalRandomMaxOffset))
  (bind _returnvalue (Math|Random|RandomFloatinRange 0.0 _localrandommaxoffset))
  (bind _returnvalue_2 (+ _seedzoffset_15_022bf3f04c9d45abd279a58dd7d53d08 _returnvalue))
  (bind _returnvalue_3 (Math|Vector|MakeVector 0.0 0.0 _returnvalue_2))
  (bind _returnvalue_4 (+ LocationToSpawn _returnvalue_3))
  ;[1 1 1] Init
  (bind _succeeded (CallFunction|SpawnChance ActorSeedData))
  (if _succeeded
    (LocalVariables|SetHitNormalLocal HitNormal)
    (LocalVariables|SetLookAtRotLocal LookAtRotation)
    (LocalVariables|SetAlignToNormalLocal _aligntolandscapenormal_37_c6495313470a8b682c3d00b3f46aa7eb)
    (LocalVariables|SetIsOnSteepSlopeLocal IsOnSteepSlope)
    (LocalVariables|SetUseSteepSlopeOffsetLocal _usesteepslopenegativeoffset_41_0acea6d44064781e182520a011991d2f)
    (LocalVariables|SetActorSeedingMethodLocal _actorseedingspawnmethod_50_2043d7544f2ab551d89eeab6b67d10f1)
    (LocalVariables|SetLocalRandomMaxOffset _actorzoffsetrandomrange_55_e15a9b6149de31b18cc5198a8b6c5631)
    ;[1 1 1] What Actor To Spawn
    (bind _actortospawn (CallFunction|SpawnActorOrdering _actorstoseed_51_d7f231c2419ae55740a8b4bcaac1f7ca ActorSeedData))
    (bind _output_get (LocalVariables|SetActorClassToSpawnLocal _actortospawn))
    (Utilities|Casting|CastToBuildingPropClass _output_get
      (:then
        (LocalVariables|SetIsBuildingPropActor true))
      (:CastFailed
        (LocalVariables|SetIsBuildingPropActor false)))
    ;[1 1 1] Z Offset
    (if (LocalVariables|GetIsOnSteepSlopeLocal)
      (if (LocalVariables|GetUseSteepSlopeOffsetLocal)
        (LocalVariables|SetFinalLocationLocal (+ _returnvalue_4 (Math|Vector|MakeVector 0.0 0.0 (Variables|ActorSeedingAdvanced|GetSteepSlopeNegativeOffsetAmount))))
        (else
          (LocalVariables|SetFinalLocationLocal _returnvalue_4)))
      ;[1 1 1] Pre Spawn Yaw Rotation
      (if (>= (Variables|InternalVariables|GetCurrentYawIndex) (Utilities|Array|LastIndex (Variables|InternalVariables|GetSpawnYawRotationsArray)))
        (Variables|InternalVariables|SetCurrentYawIndex 0)
        (else
          _))
      ;[1 1 1] Pre spawn scale
      (if _usemanualoverridescaling_27_8750023c414571fe715d07879086ddd8
        (LocalVariables|SetFinalScaleLocal (Math|Conversions|ToVector(Float) (Math|Random|RandomFloatinRange _manualoverrideminactorscale_24_cd532bb84c6429e2b0cb79910b71f3b6 _manualoverridemaxactorscale_25_7beaf98440627e0b3eb6aa9af5f5a1c7)))
        (elif (LocalVariables|GetIsBuildingPropActor)
          (LocalVariables|SetUseFinalPlacementScaleRange true)
          (else
            (LocalVariables|SetFinalScaleLocal "1.000000,1.000000,1.000000")
            ;[1 1 1] Align to landscape normal
            (if (LocalVariables|GetAlignToNormalLocal)
              (LocalVariables|SetFinalActorRotLocal (Math|Rotator|MakeRotator (.roll (LocalVariables|GetLookAtRotLocal)) (+ (.pitch (LocalVariables|GetLookAtRotLocal)) -9e+01) (.yaw (LocalVariables|GetLookAtRotLocal))))
              (else
                (LocalVariables|SetFinalActorRotLocal (Math|Rotator|MakeRotator 0.0 0.0 (Utilities|Array|Get(acopy) (Variables|InternalVariables|GetSpawnYawRotationsArray) (Variables|InternalVariables|GetCurrentYawIndex))))))
            ;[1 1 1] Random Rot Tilt
            (bind _rotout (CallFunction|RandomTilt (LocalVariables|GetFinalActorRotLocal) _actortiltdegreesrange_63_950ce81f4955d3a2b7c87cabef554dad))
            (LocalVariables|SetFinalActorRotLocal _rotout)
            ;[0.2581 1 0.359] Spawn Actor
            (bind _returnvalue_5 (Game|SpawnActorfromClass (LocalVariables|GetActorClassToSpawnLocal) (Math|Transform|MakeTransform (LocalVariables|GetFinalLocationLocal) (LocalVariables|GetFinalActorRotLocal) (LocalVariables|GetFinalScaleLocal)) "AlwaysSpawn" "SelectDefaultAtRuntime" 0 0))
            ;[1 1 1] Material Instance
            (CallFunction|SetMaterialInstanceDynamic _returnvalue_5 _actorsmaterialinstanceoverride_60_a5b63aac4ca0d3d831d255acac335632_2)
            ;[1 1 1] add references
            (Utilities|Array|Add (Variables|ActorRefs|GetAllSpawnedActorsRef) _returnvalue_5)
            (Variables|ActorRefs|SetPreviousSpawnedActorRef _returnvalue_5)
            (Utilities|Array|Add (Variables|ActorRefs|GetAllDynamicActorsInSeedingRadius) _returnvalue_5)
            ;[1 1 1] Post Spawn Yaw if aligned to normal
            (if (LocalVariables|GetAlignToNormalLocal)
              (Transformation|AddActorLocalRotation _returnvalue_5 (Math|Rotator|MakeRotator 0.0 0.0 (Utilities|Array|Get(acopy) (Variables|InternalVariables|GetSpawnYawRotationsArray) (Variables|InternalVariables|GetCurrentYawIndex))) false false))
            (Math|Integer|IncrementInt (Variables|InternalVariables|GetCurrentYawIndex)
              (:%20%20
                ;[1 1 1] post spawn building prop scale random range
                (if (LocalVariables|GetUseFinalPlacementScaleRange)
                  (bind _asbuilding_prop (Utilities|Casting|CastToBuildingProp _returnvalue_5))
                  (if (Variables|/Script/FortniteGame.BuildingSMActor|GetbRandomScaleOnPlacement _asbuilding_prop)
                    (Transformation|SetActorScale3D _returnvalue_5 (Math|Conversions|ToVector(Float) (Math|Random|RandomFloatinRange (Math|Vector2D|BreakVector2D (Variables|/Script/FortniteGame.BuildingSMActor|GetRandomScaleRange _asbuilding_prop)) (.Y (Math|Vector2D|BreakVector2D (Variables|/Script/FortniteGame.BuildingSMActor|GetRandomScaleRange _asbuilding_prop)))))))))))))
      (else
        _))))

(fn SpawnActorOrdering
  (params (ActorsArray :type Array%20of%20Actor%20Class%20References) (SeedData :type Actor%20Seed%20Data%20Structure))
  (bind (_actorstoseed_51_d7f231c2419ae55740a8b4bcaac1f7ca _mindistancetoselftype_7_01c82dcc4409c0b9f271018cdd0d768c _mindistancetobuildingactors_9_923c5abd4707bc47c299a0abab7a1749 _allowseedinwaterbodies_12_4f15c8b44c829f7843658484951353b0 _seedzoffset_15_022bf3f04c9d45abd279a58dd7d53d08 _actormaxslopeangle_18_74ca4fea488b5facee55608578cf6572 _manualoverrideminactorscale_24_cd532bb84c6429e2b0cb79910b71f3b6 _manualoverridemaxactorscale_25_7beaf98440627e0b3eb6aa9af5f5a1c7 _usemanualoverridescaling_27_8750023c414571fe715d07879086ddd8 _physicsmaterialignorelist_32_f5b884bb420973bacbce9ba20ffcc3b8 _aligntolandscapenormal_37_c6495313470a8b682c3d00b3f46aa7eb _usesteepslopenegativeoffset_41_0acea6d44064781e182520a011991d2f _actorseedingchance_45_2e134bd5489fa9e2966073a71e6d1185 _actorseedingspawnmethod_50_2043d7544f2ab551d89eeab6b67d10f1) (Utilities|Struct|BreakActorSeedData SeedData))
  (switch Utilities|FlowControl|Switch|SwitchonActorSeedingSpawnMethod _actorseedingspawnmethod_50_2043d7544f2ab551d89eeab6b67d10f1
    (:NewEnumerator0
      (Utilities|Array|Shuffle _actorstoseed_51_d7f231c2419ae55740a8b4bcaac1f7ca)
      (LocalVariables|SetActorSelectedLocal (Utilities|Array|Get(acopy) _actorstoseed_51_d7f231c2419ae55740a8b4bcaac1f7ca)))
    (:NewEnumerator1
      (if (> (+ (Variables|InternalVariables|GetCurrentActorToSpawnIndex) 1) (Utilities|Array|LastIndex _actorstoseed_51_d7f231c2419ae55740a8b4bcaac1f7ca))
        (bind _output_get (Variables|InternalVariables|SetCurrentActorToSpawnIndex 0))
        (LocalVariables|SetActorSelectedLocal (Utilities|Array|Get(acopy) _actorstoseed_51_d7f231c2419ae55740a8b4bcaac1f7ca _output_get))
        (else
          (bind _result (Math|Integer|IncrementInt (Variables|InternalVariables|GetCurrentActorToSpawnIndex))
            (:%20%20
              (LocalVariables|SetActorSelectedLocal (Utilities|Array|Get(acopy) _actorstoseed_51_d7f231c2419ae55740a8b4bcaac1f7ca _result))))))))
  (return (LocalVariables|GetActorSelectedLocal)))

(fn SpawnChance
  (params (SeedingData :type Actor%20Seed%20Data%20Structure))
  (bind (_actorstoseed_51_d7f231c2419ae55740a8b4bcaac1f7ca _mindistancetoselftype_7_01c82dcc4409c0b9f271018cdd0d768c _mindistancetobuildingactors_9_923c5abd4707bc47c299a0abab7a1749 _allowseedinwaterbodies_12_4f15c8b44c829f7843658484951353b0 _seedzoffset_15_022bf3f04c9d45abd279a58dd7d53d08 _actormaxslopeangle_18_74ca4fea488b5facee55608578cf6572 _manualoverrideminactorscale_24_cd532bb84c6429e2b0cb79910b71f3b6 _manualoverridemaxactorscale_25_7beaf98440627e0b3eb6aa9af5f5a1c7 _usemanualoverridescaling_27_8750023c414571fe715d07879086ddd8 _physicsmaterialignorelist_32_f5b884bb420973bacbce9ba20ffcc3b8 _aligntolandscapenormal_37_c6495313470a8b682c3d00b3f46aa7eb _usesteepslopenegativeoffset_41_0acea6d44064781e182520a011991d2f _actorseedingchance_45_2e134bd5489fa9e2966073a71e6d1185) (Utilities|Struct|BreakActorSeedData SeedingData))
  (bind _dropresult (Class|MissionBlueprintFunctionLibrary|DropChance (* (Math|Float|Clamp(Float) _actorseedingchance_45_2e134bd5489fa9e2966073a71e6d1185) 1e+02)))
  (return _dropresult))

(fn SpawnFirstActor
  (bind (_succeeded _validhitlocation _validhitnormal _lookatrotation _onsteepslope) (CallFunction|FindLocationforSeeding (Variables|ActorSeeding|GetPrimaryActorSeedData) (Utilities|Struct|BreakActorSeedingProfileData (Variables|ActorSeeding|GetPlacementSeedingData)) self)) ; find a location within spawn radius on a valid slope
  (if _succeeded
    (bind _validforspawn (CallFunction|ProximityandOverlapCheck _validhitlocation (Variables|ActorSeeding|GetPrimaryActorSeedData)))
    (if _validforspawn
      (CallFunction|SpawnActorinWorld _validhitlocation _validhitnormal (Variables|ActorSeeding|GetPrimaryActorSeedData) _lookatrotation _onsteepslope)
      (else
        (CallFunction|DebugPrints "Can't find valid spawn location due to proximity and overlap check failure")
        (LocalVariables|SetSucceededLocal false)))
    (return (LocalVariables|GetSucceededLocal))
    (else
      (CallFunction|DebugPrints "Can't find valid spawn location before proximity and overlap check"))))

(fn ConstructionScript
  (bind _placementseedingdata (Variables|ActorSeeding|GetPlacementSeedingData))
  (bind _seedingradius_2_816d77b142be0927697e518f6673b0f1 (Utilities|Struct|BreakActorSeedingProfileData _placementseedingdata))
  (bind _returnvalue (* _seedingradius_2_816d77b142be0927697e518f6673b0f1 2.0))
  (bind _actualbillboard (Variables|Default|GetActualBillboard))
  (bind _returnvalue_2 (Rendering|Material|CreateDynamicMaterialInstance (Variables|Default|GetEditorHelperMesh) 0 0 "None" :!class /Script/Engine.PrimitiveComponent))
  (Variables|Default|SetEditorMaterialInstanceRef _returnvalue_2)
  (bind _valid (CallFunction|ValidateData))
  (if _valid
    (Rendering|Components|Sprite|SetSprite _actualbillboard "/Game/2dAssets/Emoji/Season8/T-Emote-Icons-Season8-PalmTree.T-Emote-Icons-Season8-PalmTree")
    (if (> _seedingradius_2_816d77b142be0927697e518f6673b0f1 1.0)
      (Rendering|Material|SetVectorParameterValue (Variables|Default|GetEditorMaterialInstanceRef) "Color" "(R=0.203278,G=1.000000,B=0.444682,A=1.000000)" :!class /Script/Engine.MaterialInstanceDynamic)
      (Rendering|Material|SetScalarParameterValue (Variables|Default|GetEditorMaterialInstanceRef) "FadeDistance" (+ _returnvalue 2.5e+03) :!class /Script/Engine.MaterialInstanceDynamic))
    (else
      (Rendering|Components|Sprite|SetSprite _actualbillboard "/Game/UI/Foundation/Shell/Textures/T-Icon-HardError-64.T-Icon-HardError-64")
      (Rendering|Material|SetVectorParameterValue (Variables|Default|GetEditorMaterialInstanceRef) "Color" "(R=1.000000,G=0.000000,B=0.000000,A=1.000000)" :!class /Script/Engine.MaterialInstanceDynamic)))
  (Transformation|SetWorldScale3D (Variables|Default|GetEditorHelperMesh) (Math|Vector|MakeVector (.x (Math|Conversions|ToVector(Float) _returnvalue)) (.y (Math|Conversions|ToVector(Float) _returnvalue)) (Math|Float|Clamp(Float) (.z (Math|Conversions|ToVector(Float) _returnvalue)) 1.0 (Variables|ActorSeedingAdvanced|GetZProjectionDistance)))))

(fn ValidateData
  (bind _placementseedingdata (Variables|ActorSeeding|GetPlacementSeedingData))
  (bind (_seedingradius_2_816d77b142be0927697e518f6673b0f1 _numprimaryseedactors_15_a4812bae43a70605e6445396062315af) (Utilities|Struct|BreakActorSeedingProfileData _placementseedingdata))
  (bind (_seedingradius_2_816d77b142be0927697e518f6673b0f1_2 _numprimaryseedactors_15_a4812bae43a70605e6445396062315af_2 _numsecondaryseedactors_18_e5908d964e9d7e0f9b994786c762732e _numtertiaryseedactors_19_b35a07a2453f88aeacf7499d9dd7543c) (Utilities|Struct|BreakActorSeedingProfileData (Variables|ActorSeeding|GetPlacementSeedingData)))
  (if (> (Utilities|Array|Length (Utilities|Struct|BreakActorSeedData (Variables|ActorSeeding|GetPrimaryActorSeedData))) 0)
    (if (> _numprimaryseedactors_15_a4812bae43a70605e6445396062315af 0)
      (if (> _seedingradius_2_816d77b142be0927697e518f6673b0f1 1.0)
        (if (not (> (Utilities|Operators|Add _numprimaryseedactors_15_a4812bae43a70605e6445396062315af_2 _numsecondaryseedactors_18_e5908d964e9d7e0f9b994786c762732e _numtertiaryseedactors_19_b35a07a2453f88aeacf7499d9dd7543c) 300))
          (return true))
        (else
          _))
      (return false))))

(fn WaterAboveBelowCheck
  (params (LandscapeHitVector :type Vector) (WaterHitVector :type Vector))
  (return (> (.z LandscapeHitVector) (.z WaterHitVector))))
